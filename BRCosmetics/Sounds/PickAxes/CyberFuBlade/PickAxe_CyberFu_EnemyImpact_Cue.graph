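# Sound cue for the CyberFu Blade pickaxe's enemy impact in Battle Royale cosmetics.
# A short fixed delay feeds a random node that picks one of five equally weighted impact waves.
# Playback uses the shared pickaxe impact attenuation, impact sound class and cosmetic property sheet, and primes on load.
# Overridden concurrency limits simultaneous plays per owner, stopping the quietest.
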SoundCue /BRCosmetics/Sounds/PickAxes/CyberFuBlade/PickAxe_CyberFu_EnemyImpact_Cue
root=Delay_1
nodes [
 {id=Delay_1, children [Random_0], properties {delayMax=0.1, delayMin=0.1}}
 {id=Random_0, children [WavePlayer_0,WavePlayer_1,WavePlayer_2,WavePlayer_8,WavePlayer_9], properties {weights [1,1,1,1,1]}}
 {id=WavePlayer_0, properties {soundWaveAssetPtr=/FNGameplayCues/Athena/Sounds/Weapons/PickAxes/CyberFuBlade/PickAxe_CyberFu_EnemyImpact_01}}
 {id=WavePlayer_1, properties {soundWaveAssetPtr=/FNGameplayCues/Athena/Sounds/Weapons/PickAxes/CyberFuBlade/PickAxe_CyberFu_EnemyImpact_02}}
 {id=WavePlayer_2, properties {soundWaveAssetPtr=/FNGameplayCues/Athena/Sounds/Weapons/PickAxes/CyberFuBlade/PickAxe_CyberFu_EnemyImpact_03}}
 {id=WavePlayer_8, properties {soundWaveAssetPtr=/FNGameplayCues/Athena/Sounds/Weapons/PickAxes/CyberFuBlade/PickAxe_CyberFu_EnemyImpact_05}}
 {id=WavePlayer_9, properties {soundWaveAssetPtr=/FNGameplayCues/Athena/Sounds/Weapons/PickAxes/CyberFuBlade/PickAxe_CyberFu_EnemyImpact_06}}
]
properties {
 attenuationSettings=/Game/Sounds/Attenuation/Fort_Pickaxe_Impact_Att
 audioPropertiesSheet=/BRCosmetics/Sounds/Settings/PropertySheets/Pickaxe/APS_Cosmetic_Pickaxe_Impact
 bOverrideConcurrency=true
 bPrimeOnLoad=true
 maxDistance=4300
 soundClassObject=&Fort_Impacts
 volumeMultiplier=0.7125
 attenuationOverrides {
  bAttenuateWithLPF=true
  distanceAlgorithm=Logarithmic
  falloffDistance=800
  lPFRadiusMax=800
  lPFRadiusMin=400
  attenuationShapeExtents {x=300}
 }
 concurrencyOverrides {bLimitToOwner=true, maxCount=2, resolutionRule=StopQuietest}
 concurrencySet [/Game/Sounds/Concurrency_Rules/OneMaximum_Owner_StopOldest]
}
